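# Sound cue for tennis ball physics impacts, played with physics impact attenuation in the impacts sound class.
# An ImpactType switch routes its first three inputs to random bounce waves and the last to random impact waves.
# Each branch scales volume and pitch continuously by the Strength parameter, with inverted pitch output so stronger hits sound lower.
# A root modulator adds slight random pitch and volume variation.
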SoundCue /FNGameplayCues/Athena/Sounds/Props/TennisBall/PROP_TennisBall_Impact_Phys_Cue
root=Modulator_1
nodes [
 {
  id=ModulatorContinuous_0
  children [Random_3]
  properties {
   pitchModulationParams {default=0, maxOutput=0.8, minOutput=1.2, parameterName=Strength}
   volumeModulationParams {default=0, minOutput=0.2, parameterName=Strength}
  }
 }
 {
  id=ModulatorContinuous_1
  children [Random_3]
  properties {
   pitchModulationParams {default=0, maxOutput=0.8, minOutput=1.2, parameterName=Strength}
   volumeModulationParams {default=0, maxOutput=1.2, minOutput=0.8, parameterName=Strength}
  }
 }
 {
  id=ModulatorContinuous_2
  children [Random_4]
  properties {
   pitchModulationParams {default=0, maxOutput=0.8, minOutput=1.2, parameterName=Strength}
   volumeModulationParams {default=0, maxOutput=1.2, minOutput=0.8, parameterName=Strength}
  }
 }
 {id=Modulator_1, children [Switch_0], properties {pitchMax=1.1, pitchMin=0.9, volumeMax=1.1, volumeMin=0.9}}
 {id=Random_3, children [WavePlayer_23, WavePlayer_24, WavePlayer_25], properties {weights [1,1,1]}}
 {id=Random_4, children [WavePlayer_26, WavePlayer_27, WavePlayer_28], properties {weights [1,1,1]}}
 {
  id=Switch_0
  children [ModulatorContinuous_0, ModulatorContinuous_0, ModulatorContinuous_1, ModulatorContinuous_2]
  properties {intParameterName=ImpactType}
 }
 {id=WavePlayer_23, properties {soundWaveAssetPtr=&/PROP_TennisBall_Bounce_01}}
 {id=WavePlayer_24, properties {soundWaveAssetPtr=&/PROP_TennisBall_Bounce_02}}
 {id=WavePlayer_25, properties {soundWaveAssetPtr=&/PROP_TennisBall_Bounce_03}}
 {id=WavePlayer_26, properties {soundWaveAssetPtr=&/PROP_TennisBall_Impact_01}}
 {id=WavePlayer_27, properties {soundWaveAssetPtr=&/PROP_TennisBall_Impact_02}}
 {id=WavePlayer_28, properties {soundWaveAssetPtr=&/PROP_TennisBall_Impact_03}}
]
properties {
 attenuationSettings=/Game/Sounds/Physics/Settings/Physics_Impact_Att
 maxDistance=8512
 pitchMultiplier=1.05
 soundClassObject=&Fort_Impacts
 volumeMultiplier=0.7
}
